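# Niagara effect for the zipline pulley, throwing hot sparks, skid streaks and smoke from two offset sides, with an orange flickering light.
# Paired emitters cover each side: mesh skids, sprite smoke, and short-lived plus longer drifting GPU sparks stretched by speed.
# User.SpawnSpark drives spawn probability or rate for everything except the light.
# Emitter-age curves ramp up spark launch speed and light brightness after start.
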NiagaraSystem /Game/Effects/Fort_Effects/Gameplay/Zipline/NS_Zipline_Pulley_SpeedLines_Converted
dependencies {
 usedDataInterfaces [&NiagaraDataInterfaceCurve, &NiagaraDataInterfaceVectorCurve]
 usedDynamicInputs []
 usedModules [
  &self:SystemSpawnScript.SystemScriptSource.SystemScriptGraph.NiagaraNodeAssignment_0.SetVariables_#1
  &SystemState
 ]
}
emitters {
 Light {
  emitterUpdateScript [
   &EmitterState(
    Inactive Response=::NewEnumerator1
    Life Cycle Mode=::NewEnumerator1
    Loop Behavior=::NewEnumerator1
    Loop Duration Mode=::NewEnumerator1
    UseLoopDelay=true
   )
   &SpawnBurst_Instantaneous(Use Spawn Probability=false, UseLoopCountLimit=false)
  ]
  particleSpawnScript [
   &InitializeParticle(Color Mode=::NewEnumerator0, Lifetime=0.5, UsePositionOffset=false, Write Material Random=true)
   &SystemLocation(MaskSpawnGroup=false)
  ]
  particleUpdateScript [
   &ParticleState(Kill Particles When Lifetime Has Elapsed=false, Let Infinitely Lived Particles Die When Emitter Deactivates=true)
   &FN_Light_Attributes(
    Color=<rgba 1 0.549 0.1137 1>
    Create Color Attribute=true
    Scale Color=&Multiply_Float(
     A=&FloatFromCurve(
      FloatCurve=DataInterface() {propertyValues {Curve {keys [{interpMode=RCIM_Cubic, leaveTangent=2.9182, tangentMode=RCTM_Break, time=0, value=0}, <at 0.5 1>]}}}
      CurveIndex=$Emitter.Age
      Scale Curve=6
     )
     B=&RandomRangeFloat(Minimum=0.6, Maximum=1, Evaluation Type=::NewEnumerator1)
    )
   )
  ]
  rendererClasses [&NiagaraLightRendererProperties]
  renderers [{rendererClass=&NiagaraLightRendererProperties, rendererIndex=0}]
 }
 SkidSharp {
  emitterUpdateScript [
   &EmitterState(Life Cycle Mode=::NewEnumerator1, Loop Duration=1, UseLoopDelay=true)
   &SpawnRate(Spawn Probability=$User.SpawnSpark, SpawnRate=15, Use Spawn Probability=true)
  ]
  particleSpawnScript [
   /Niagara/Modules/Spawn/Initialization/InitializeParticle(
    Lifetime=&RandomRangeFloat(Minimum=0.12, Maximum=0.16, Evaluation Type=::NewEnumerator1)
    Mesh Scale=<0.8 0.8 0.8>
    Module.Write Color=true
    Module.Write Lifetime=true
    Module.Write Mass=true
    Module.Write Material Random=true
    Module.Write Position=false
    Module.Write Scale=true
    Module.Write SpriteRotation=false
    Module.Write SpriteSize=false
    Position=&RandomRangeVector(Minimum=<-22 0 0>, Maximum=<22 0 0>, Evaluation Type=::NewEnumerator1)
   )
   &Color(Color=&MakeLinearColorFromVectorAndFloat(
     "Vector (RGB)"=&MakeVectorFromLinearColorRGB(Color=<rgba 1 0.3739 0.1007 0.6>)
     "Float (Alpha)"=&ClampFloat(Float=&RandomRangeFloat(Minimum=1.2, Maximum=2, Evaluation Type=::NewEnumerator1))
    ))
   &AddVelocity(Velocity=&RandomRangeVector(Minimum=<0 -25 0>, Maximum=<0 -5 0>, Evaluation Type=::NewEnumerator1))
   &DynamicMaterialParameters(
    Index 0 Param 3=&RandomRangeFloat(Minimum=1, Maximum=1.5, Evaluation Type=::NewEnumerator1)
    Module.Param0WriteEnabled=false
    Module.Param1WriteEnabled=false
    Module.Param2WriteEnabled=true
    Module.Param3WriteEnabled=false
   ) {module=DynamicMaterialParameters001}
   &InheritVelocity(Module.Limit Speed=false, Velocity Limit=0, Velocity Scale=<-0.1 -0.1 0>) {enabled=false}
   &SystemLocation(MaskSpawnGroup=false, Offset=<0 13 3>)
   &InitialMeshOrientation(
    Rotation=&RandomRangeVector(Minimum=<0 0 0.1>, Maximum=<0 -0.1 -0.1>)
    Use Orientation Vector=true
    Use Rotation Vector=true
   )
   &AddVelocityFromPoint(Module.Use Falloff Distance=false, Velocity Strength=&RandomRangeFloat(Minimum=120, Maximum=250))
  ]
  particleUpdateScript [
   /SetVariables(Particles.SpriteSize=<1 1>) {isSetParameters=true, module=SetVariables_#2}
   &ParticleState()
   &Drag(Drag=4, Use Linear Drag=true, Use Rotational Drag=false)
   &SolveForcesAndVelocity(Clamp Velocity=false, Limit Acceleration=false)
   &ScaleColor(
    Scale Alpha=&FloatFromCurve(FloatCurve=DataInterface() {propertyValues {Curve {keys [<at 0 1 RCIM_Cubic>, {interpMode=RCIM_Cubic, leaveTangent=-3.0416, time=1, value=0}]}}})
    ScaleA=true
    ScaleRGB=true
   )
   &DynamicMaterialParameters(
    Index 0 Param 1=&FloatFromCurve(FloatCurve=DataInterface() {propertyValues {Curve {keys [<at 0.1 0.1 RCIM_Cubic>,<at 0.8 1>]}}})
    Index 0 Param 2=2
    Index 0 Param 4=&FloatFromCurve(FloatCurve=DataInterface() {propertyValues {Curve {keys [<at 0 1>,<at 1 3>]}}})
    Module.Param0WriteEnabled=true
    Module.Param1WriteEnabled=true
    Module.Param2WriteEnabled=false
    Module.Param3WriteEnabled=true
   )
   &ScaleMeshSize(Scale Factor=&VectorFromFloat(Value=&FloatFromCurve(FloatCurve=DataInterface() {propertyValues {Curve {keys [{arriveTangent=2.1642, interpMode=RCIM_Cubic, leaveTangent=2.1642, tangentMode=RCTM_User, time=0, value=0}, <at 1 1 RCIM_Cubic>]}}})))
  ]
  rendererClasses [&NiagaraMeshRendererProperties]
  renderers [{rendererClass=&NiagaraMeshRendererProperties, rendererIndex=0}]
 }
 SkidSharp001 {
  emitterUpdateScript [
   &EmitterState(Life Cycle Mode=::NewEnumerator1, Loop Duration=1, UseLoopDelay=true)
   &SpawnRate(Spawn Probability=$User.SpawnSpark, SpawnRate=15, Use Spawn Probability=true)
  ]
  particleSpawnScript [
   &InitializeParticle(
    Lifetime=&RandomRangeFloat(Minimum=0.12, Maximum=0.16, Evaluation Type=::NewEnumerator1)
    Mesh Scale=<0.8 0.8 0.8>
    Module.Write Color=true
    Module.Write Lifetime=true
    Module.Write Mass=true
    Module.Write Material Random=true
    Module.Write Position=false
    Module.Write Scale=true
    Module.Write SpriteRotation=false
    Module.Write SpriteSize=false
    Position=&RandomRangeVector(Minimum=<-22 0 0>, Maximum=<22 0 0>, Evaluation Type=::NewEnumerator1)
   )
   &Color(Color=&MakeLinearColorFromVectorAndFloat(
     "Vector (RGB)"=&MakeVectorFromLinearColorRGB(Color=<rgba 1 0.3739 0.1007 0.6>)
     "Float (Alpha)"=&ClampFloat(Float=&RandomRangeFloat(Minimum=1.2, Maximum=2, Evaluation Type=::NewEnumerator1))
    ))
   &AddVelocity(Velocity=&RandomRangeVector(Minimum=<0 -25 0>, Maximum=<0 -5 0>, Evaluation Type=::NewEnumerator1)) {enabled=false}
   &DynamicMaterialParameters(
    Index 0 Param 3=&RandomRangeFloat(Minimum=1, Maximum=1.5, Evaluation Type=::NewEnumerator1)
    Module.Param0WriteEnabled=false
    Module.Param1WriteEnabled=false
    Module.Param2WriteEnabled=true
    Module.Param3WriteEnabled=false
   ) {module=DynamicMaterialParameters001}
   &InheritVelocity(Module.Limit Speed=false, Velocity Limit=0, Velocity Scale=<0 0 0>)
   &SystemLocation(MaskSpawnGroup=false, Offset=<0 -12 7>)
   &AddVelocityFromPoint(Module.Use Falloff Distance=false, Velocity Strength=&RandomRangeFloat(Minimum=-30, Maximum=-70))
   &InitialMeshOrientation(
    Rotation=&RandomRangeVector(Minimum=<0 0 0.15>, Maximum=<0 -0.1 0.25>)
    Use Orientation Vector=true
    Use Rotation Vector=true
   )
  ]
  particleUpdateScript [
   /SetVariables(Particles.SpriteSize=<1 1>) {isSetParameters=true, module=SetVariables_#2}
   &ParticleState()
   &Drag(Drag=4, Use Linear Drag=true, Use Rotational Drag=false)
   &SolveForcesAndVelocity(Clamp Velocity=false, Limit Acceleration=false)
   &ScaleColor(
    Scale Alpha=&FloatFromCurve(FloatCurve=DataInterface() {propertyValues {Curve {keys [<at 0 1 RCIM_Cubic>, {interpMode=RCIM_Cubic, leaveTangent=-3.0416, time=1, value=0}]}}})
    ScaleA=true
    ScaleRGB=true
   )
   &DynamicMaterialParameters(
    Index 0 Param 1=&FloatFromCurve(FloatCurve=DataInterface() {propertyValues {Curve {keys [<at 0.1 0.1 RCIM_Cubic>,<at 0.8 1>]}}})
    Index 0 Param 2=2
    Index 0 Param 4=&FloatFromCurve(FloatCurve=DataInterface() {propertyValues {Curve {keys [<at 0 1>,<at 1 3>]}}})
    Module.Param0WriteEnabled=true
    Module.Param1WriteEnabled=true
    Module.Param2WriteEnabled=false
    Module.Param3WriteEnabled=true
   )
   &ScaleMeshSize(Scale Factor=&VectorFromFloat(Value=&FloatFromCurve(FloatCurve=DataInterface() {propertyValues {Curve {keys [{arriveTangent=2.1642, interpMode=RCIM_Cubic, leaveTangent=2.1642, tangentMode=RCTM_User, time=0, value=0}, <at 1 1 RCIM_Cubic>]}}})))
  ]
  rendererClasses [&NiagaraMeshRendererProperties]
  renderers [{rendererClass=&NiagaraMeshRendererProperties, rendererIndex=0}]
 }
 Smoke {
  emitterUpdateScript [
   &EmitterState()
   &SpawnRate(Spawn Probability=$User.SpawnSpark, SpawnRate=14, Use Spawn Probability=true)
  ]
  particleSpawnScript [
   /Niagara/Modules/Spawn/Initialization/V2/InitializeParticle(
    Color Mode=::NewEnumerator0
    Lifetime=&RandomRangeFloat(Minimum=0.2, Maximum=0.4)
    Sprite Rotation Angle=&RandomRangeFloat(Minimum=0, Maximum=1)
    Sprite Rotation Mode=::NewEnumerator1
    Sprite Size Mode=::NewEnumerator1
    Uniform Sprite Size Max=25
    Uniform Sprite Size Min=18
    UsePositionOffset=false
    Write Material Random=true
   )
   &SphereLocation(
    MaskSpawnGroup=false
    Override Seed=false
    Sphere Radius=&RandomRangeFloat(Minimum=3, Maximum=6)
    Surface Only=false
    UseNonUniformScale=false
    UseOffset=false
    UseSphereOrientation=false
   )
   &AddVelocity(Velocity=&Multiply_VectorByFloat(Vector=$Output.SphereLocation.SphereVector, Float=&RandomRangeFloat(Minimum=-10, Maximum=10)))
   &AddVelocity(Velocity=&RandomRangeVector(Minimum=<0 50 0>, Maximum=<1 100 1>)) {module=AddVelocity001}
   &SystemLocation(MaskSpawnGroup=false, Offset=<0 20 7>)
   &Color(Color=<rgba 0.2083 0.1842 0.1584 1.4>)
  ]
  particleUpdateScript [
   &SolveForcesAndVelocity(Clamp Velocity=false, Limit Acceleration=false)
   &ParticleState()
   &DynamicMaterialParameters(
    Index 0 Param 1=&FloatFromCurve(FloatCurve=DataInterface() {propertyValues {Curve {keys [<at 0 0>,<at 1 0.6>]}}})
    Module.Param0WriteEnabled=true
    Module.Param1WriteEnabled=true
    Module.Param2WriteEnabled=true
    Module.Param3WriteEnabled=true
   )
   &ScaleColor(
    Scale Alpha=&FloatFromCurve(FloatCurve=DataInterface() {propertyValues {Curve {keys [<at 0 0>,<at 0.2037 1>,<at 1 0>]}}})
    ScaleA=true
    ScaleRGB=true
   )
   &ScaleSpriteSize(Scale Factor=&Vector2DFromFloat(Value=&FloatFromCurve(FloatCurve=DataInterface() {propertyValues {Curve {keys [{arriveTangent=1.9106, interpMode=RCIM_Cubic, leaveTangent=1.9106, tangentMode=RCTM_User, time=0, value=0}, <at 1 1 RCIM_Cubic>]}}})))
  ]
  rendererClasses [&NiagaraSpriteRendererProperties]
  renderers [{rendererClass=&NiagaraSpriteRendererProperties}]
 }
 Smoke001 {
  emitterUpdateScript [
   &EmitterState()
   &SpawnRate(Spawn Probability=$User.SpawnSpark, SpawnRate=14, Use Spawn Probability=true)
  ]
  particleSpawnScript [
   &InitializeParticle(
    Color Mode=::NewEnumerator0
    Lifetime=&RandomRangeFloat(Minimum=0.2, Maximum=0.4)
    Sprite Rotation Angle=&RandomRangeFloat(Minimum=0, Maximum=1)
    Sprite Rotation Mode=::NewEnumerator1
    Sprite Size Mode=::NewEnumerator1
    Uniform Sprite Size Max=25
    Uniform Sprite Size Min=18
    UsePositionOffset=false
    Write Material Random=true
   )
   &SphereLocation(
    MaskSpawnGroup=false
    Override Seed=false
    Sphere Radius=&RandomRangeFloat(Minimum=3, Maximum=6)
    Surface Only=false
    UseNonUniformScale=false
    UseOffset=false
    UseSphereOrientation=false
   )
   &AddVelocity(Velocity=&Multiply_VectorByFloat(Vector=$Output.SphereLocation.SphereVector, Float=&RandomRangeFloat(Minimum=-10, Maximum=10)))
   &AddVelocity(Velocity=&RandomRangeVector(Minimum=<0 50 0>, Maximum=<1 100 1>)) {module=AddVelocity001}
   &SystemLocation(MaskSpawnGroup=false, Offset=<0 -18 7>)
   &Color(Color=<rgba 0.2083 0.1842 0.1584 1.4>)
  ]
  particleUpdateScript [
   &SolveForcesAndVelocity(Clamp Velocity=false, Limit Acceleration=false)
   &ParticleState()
   &DynamicMaterialParameters(
    Index 0 Param 1=&FloatFromCurve(FloatCurve=DataInterface() {propertyValues {Curve {keys [<at 0 0>,<at 1 0.6>]}}})
    Module.Param0WriteEnabled=true
    Module.Param1WriteEnabled=true
    Module.Param2WriteEnabled=true
    Module.Param3WriteEnabled=true
   )
   &ScaleColor(
    Scale Alpha=&FloatFromCurve(FloatCurve=DataInterface() {propertyValues {Curve {keys [<at 0 0>,<at 0.2037 1>,<at 1 0>]}}})
    ScaleA=true
    ScaleRGB=true
   )
   &ScaleSpriteSize(Scale Factor=&Vector2DFromFloat(Value=&FloatFromCurve(FloatCurve=DataInterface() {propertyValues {Curve {keys [{arriveTangent=1.9106, interpMode=RCIM_Cubic, leaveTangent=1.9106, tangentMode=RCTM_User, time=0, value=0}, <at 1 1 RCIM_Cubic>]}}})))
  ]
  rendererClasses [&NiagaraSpriteRendererProperties]
  renderers [{rendererClass=&NiagaraSpriteRendererProperties}]
 }
 SparksStart {
  simTarget=GPUComputeSim
  emitterUpdateScript [
   &EmitterState(Life Cycle Mode=::NewEnumerator1, Loop Duration=1, UseLoopDelay=true)
   &SpawnBurst_Instantaneous(
    Spawn Count=&RandomRangeInt(Minimum=35, Maximum=60)
    Spawn Probability=$User.SpawnSpark
    Use Spawn Probability=true
   )
   &SpawnRate(SpawnRate=&Multiply_Float(A=100, B=$User.SpawnSpark), Use Spawn Probability=false)
  ]
  particleSpawnScript [
   /Niagara/Modules/Spawn/Initialization/InitializeParticle(
    Lifetime=&RandomRangeFloat(Minimum=0.12, Maximum=0.4, Evaluation Type=::NewEnumerator1)
    Module.Write Color=true
    Module.Write Lifetime=true
    Module.Write Mass=true
    Module.Write Material Random=true
    Module.Write Position=false
    Module.Write Scale=false
    Module.Write SpriteRotation=false
    Module.Write SpriteSize=true
    Position=&RandomRangeVector(Minimum=<-22 0 0>, Maximum=<22 0 0>, Evaluation Type=::NewEnumerator1)
    Sprite Size=&RandomRangeVector2D(Minimum=<0.5 0.75>, Maximum=<1 3>, Evaluation Type=::NewEnumerator1)
   )
   /Niagara/Modules/Spawn/Location/SphereLocation(
    Hemisphere X=true
    MaskSpawnGroup=false
    Module.Override Randomness=false
    Sphere Radius=&RandomRangeFloat(Minimum=5, Maximum=10, Evaluation Type=::NewEnumerator1)
    Surface Only=false
   )
   &Color(Color=&MakeLinearColorFromVectorAndFloat(
     "Vector (RGB)"=&MakeVectorFromLinearColorRGB(Color=<rgba 5 1.6977 0.0741 1>)
     "Float (Alpha)"=&ClampFloat(Float=1)
    ))
   &SystemLocation(MaskSpawnGroup=false, Offset=&RandomRangeVector(Minimum=<-1 20 4>, Maximum=<1 20 4>))
   &AddVelocityFromPoint(
    Module.Use Falloff Distance=false
    Velocity Strength=&FloatFromCurve(
     FloatCurve=DataInterface() {propertyValues {Curve {keys [{interpMode=RCIM_Cubic, tangentMode=RCTM_Break, time=0, value=0.3423}, <at 1 1>]}}}
     CurveIndex=$Emitter.Age
     Scale Curve=&RandomRangeFloat(Minimum=130, Maximum=580)
    )
   )
   &AddVelocity(Velocity=&Multiply_VectorByFloat(
     Vector=$Output.SphereLocation.SphereVector
     Float=&RandomRangeFloat(Minimum=-14, Maximum=30, Evaluation Type=::NewEnumerator1)
    )) {module=AddVelocity002}
  ]
  particleUpdateScript [
   &ParticleState()
   &AccelerationForce(Acceleration=<0 0 -100>) {enabled=false}
   &GravityForce(Gravity=&RandomRangeVector(Minimum=<0 0 -200>, Maximum=<0 0 0>))
   &SolveForcesAndVelocity(Clamp Velocity=false, Limit Acceleration=false)
   &ScaleSpriteSize(Scale Factor=&Vector2DFromFloat(Value=&FloatFromCurve(FloatCurve=DataInterface() {propertyValues {Curve {keys [
          {arriveTangent=55.4408, interpMode=RCIM_Cubic, time=0, value=0}
          <at 0.03 1 RCIM_Cubic>
          {arriveTangent=-3.3237, leaveTangent=-3.3237, time=1, value=0}
         ]}}})))
   &ScaleColor(
    Scale Alpha=&FloatFromCurve(FloatCurve=DataInterface() {propertyValues {Curve {keys [{arriveTangent=-6.8352, interpMode=RCIM_Cubic, time=0, value=2}, <at 0.2 1>]}}})
    Scale RGB=&VectorFromCurve(VectorCurve=DataInterface() {propertyValues {
       XCurve {keys [<at 0 16 RCIM_Cubic>,<at 0.8 1 RCIM_Cubic>]}
       YCurve {keys [<at 0 16 RCIM_Cubic>,<at 0.8 0.1249 RCIM_Cubic>]}
       ZCurve {keys [<at 0 16 RCIM_Cubic>,<at 0.8 0.2547 RCIM_Cubic>]}
      }})
    ScaleA=true
    ScaleRGB=true
   )
   &ScaleSpriteSizeBySpeed(Max Scale Factor=<1 50>, Source Velocity:Vector3f=&Multiply_Vector(A=<0 0.015 1>, B=$Particles.Velocity))
  ]
  rendererClasses [&NiagaraSpriteRendererProperties]
  renderers [{rendererClass=&NiagaraSpriteRendererProperties}]
 }
 SparksStart001 {
  simTarget=GPUComputeSim
  emitterUpdateScript [
   &EmitterState(Life Cycle Mode=::NewEnumerator1, Loop Behavior=::NewEnumerator1, Loop Duration=3, UseLoopDelay=true)
   &SpawnBurst_Instantaneous(
    Spawn Count=&RandomRangeInt(Minimum=35, Maximum=60)
    Spawn Probability=$User.SpawnSpark
    Use Spawn Probability=true
   )
   &SpawnRate(SpawnRate=&Multiply_Float(A=45, B=$User.SpawnSpark), Use Spawn Probability=false)
  ]
  particleSpawnScript [
   &InitializeParticle(
    Lifetime=&RandomRangeFloat(Minimum=0.5, Maximum=1.5, Evaluation Type=::NewEnumerator1)
    Module.Write Color=true
    Module.Write Lifetime=true
    Module.Write Mass=true
    Module.Write Material Random=true
    Module.Write Position=false
    Module.Write Scale=false
    Module.Write SpriteRotation=false
    Module.Write SpriteSize=true
    Position=&RandomRangeVector(Minimum=<-22 0 0>, Maximum=<22 0 0>, Evaluation Type=::NewEnumerator1)
    Sprite Size=&RandomRangeVector2D(Minimum=<0.5 0.5>, Maximum=<1 1>, Evaluation Type=::NewEnumerator1)
   )
   &SphereLocation(
    Hemisphere X=true
    MaskSpawnGroup=false
    Module.Override Randomness=false
    Sphere Radius=&RandomRangeFloat(Minimum=35, Maximum=50, Evaluation Type=::NewEnumerator1)
    Surface Only=false
   )
   &Color(Color=&MakeLinearColorFromVectorAndFloat(
     "Vector (RGB)"=&MakeVectorFromLinearColorRGB(Color=<rgba 5 1.6977 0.0741 1>)
     "Float (Alpha)"=&ClampFloat(Float=1)
    ))
   &SystemLocation(MaskSpawnGroup=false, Offset=&RandomRangeVector(Minimum=<-4 16 4>, Maximum=<4 20 4>))
   &AddVelocityFromPoint(Module.Use Falloff Distance=false, Velocity Strength=&RandomRangeFloat(Minimum=60, Maximum=150)) {enabled=false}
   &AddVelocity(Coordinate Space=::World, Velocity=&RandomRangeVector(Minimum=<-10 -10 0>, Maximum=<10 10 1>)) {enabled=false, module=AddVelocity002}
   &AddVelocityFromPoint(Module.Use Falloff Distance=false, Velocity Strength=&RandomRangeFloat(Minimum=85, Maximum=200)) {enabled=false, module=AddVelocityFromPoint001}
   &AddVelocityFromPoint(
    Module.Use Falloff Distance=false
    Velocity Strength=&FloatFromCurve(
     FloatCurve=DataInterface() {propertyValues {Curve {keys [{interpMode=RCIM_Cubic, tangentMode=RCTM_Break, time=0, value=0.3423}, <at 1 1>]}}}
     CurveIndex=$Emitter.Age
     Scale Curve=&RandomRangeFloat(Minimum=90, Maximum=260)
    )
   ) {module=AddVelocityFromPoint002}
  ]
  particleUpdateScript [
   &ParticleState()
   &AccelerationForce(Acceleration=<0 0 -100>)
   &GravityForce(Gravity=&RandomRangeVector(Minimum=<0 0 -80>, Maximum=<0 0 -40>))
   &VectorNoiseForce(Module.Use Random Noise Time Interval=false)
   &SolveForcesAndVelocity(Clamp Velocity=false, Limit Acceleration=false)
   &ScaleSpriteSize(Scale Factor=&Vector2DFromFloat(Value=&FloatFromCurve(FloatCurve=DataInterface() {propertyValues {Curve {keys [
          {arriveTangent=55.4408, interpMode=RCIM_Cubic, time=0, value=0}
          <at 0.03 1 RCIM_Cubic>
          {arriveTangent=-3.3237, leaveTangent=-3.3237, time=1, value=0}
         ]}}})))
   &ScaleColor(
    Scale Alpha=&FloatFromCurve(FloatCurve=DataInterface() {propertyValues {Curve {keys [{arriveTangent=-6.8352, interpMode=RCIM_Cubic, time=0, value=2}, <at 0.2 1>]}}})
    Scale RGB=&VectorFromCurve(VectorCurve=DataInterface() {propertyValues {
       XCurve {keys [<at 0 16 RCIM_Cubic>,<at 0.8 1 RCIM_Cubic>]}
       YCurve {keys [<at 0 16 RCIM_Cubic>,<at 0.8 0.1249 RCIM_Cubic>]}
       ZCurve {keys [<at 0 16 RCIM_Cubic>,<at 0.8 0.2547 RCIM_Cubic>]}
      }})
    ScaleA=true
    ScaleRGB=true
   )
   &ScaleSpriteSizeBySpeed(Max Scale Factor=<1 50>, Source Velocity:Vector3f=&Multiply_Vector(A=<0 0.015 1>, B=$Particles.Velocity))
  ]
  rendererClasses [&NiagaraSpriteRendererProperties]
  renderers [{rendererClass=&NiagaraSpriteRendererProperties}]
 }
 SparksStart002 {
  simTarget=GPUComputeSim
  emitterUpdateScript [
   &EmitterState(Life Cycle Mode=::NewEnumerator1, Loop Behavior=::NewEnumerator1, Loop Duration=3, UseLoopDelay=true)
   &SpawnBurst_Instantaneous(
    Spawn Count=&RandomRangeInt(Minimum=35, Maximum=60)
    Spawn Probability=$User.SpawnSpark
    Use Spawn Probability=true
   )
   &SpawnRate(SpawnRate=&Multiply_Float(A=45, B=$User.SpawnSpark), Use Spawn Probability=false)
  ]
  particleSpawnScript [
   &InitializeParticle(
    Lifetime=&RandomRangeFloat(Minimum=0.5, Maximum=1.3, Evaluation Type=::NewEnumerator1)
    Module.Write Color=true
    Module.Write Lifetime=true
    Module.Write Mass=true
    Module.Write Material Random=true
    Module.Write Position=false
    Module.Write Scale=false
    Module.Write SpriteRotation=false
    Module.Write SpriteSize=true
    Position=&RandomRangeVector(Minimum=<-22 0 0>, Maximum=<22 0 0>, Evaluation Type=::NewEnumerator1)
    Sprite Size=&RandomRangeVector2D(Minimum=<0.5 0.5>, Maximum=<1 1>, Evaluation Type=::NewEnumerator1)
   )
   &SphereLocation(
    Hemisphere X=true
    MaskSpawnGroup=false
    Module.Override Randomness=false
    Sphere Radius=&RandomRangeFloat(Minimum=35, Maximum=50, Evaluation Type=::NewEnumerator1)
    Surface Only=false
   )
   &Color(Color=&MakeLinearColorFromVectorAndFloat(
     "Vector (RGB)"=&MakeVectorFromLinearColorRGB(Color=<rgba 5 1.6977 0.0741 1>)
     "Float (Alpha)"=&ClampFloat(Float=1)
    ))
   &AddVelocityFromPoint(Module.Use Falloff Distance=false, Velocity Strength=&RandomRangeFloat(Minimum=60, Maximum=150)) {enabled=false}
   &SystemLocation(MaskSpawnGroup=false, Offset=&RandomRangeVector(Minimum=<-4 -12 5>, Maximum=<4 -12 5>))
   &AddVelocity(Coordinate Space=::World, Velocity=&RandomRangeVector(Minimum=<-10 -10 0>, Maximum=<10 10 1>)) {enabled=false, module=AddVelocity002}
   &AddVelocityFromPoint(Module.Use Falloff Distance=false, Velocity Strength=&RandomRangeFloat(Minimum=-85, Maximum=-200)) {enabled=false, module=AddVelocityFromPoint001}
   &AddVelocityFromPoint(
    Module.Use Falloff Distance=false
    Velocity Strength=&FloatFromCurve(
     FloatCurve=DataInterface() {propertyValues {Curve {keys [{interpMode=RCIM_Cubic, tangentMode=RCTM_Break, time=0, value=0.3423}, <at 1 1>]}}}
     CurveIndex=$Emitter.Age
     Scale Curve=&RandomRangeFloat(Minimum=-90, Maximum=-325)
    )
   ) {module=AddVelocityFromPoint002}
  ]
  particleUpdateScript [
   &ParticleState()
   &AccelerationForce(Acceleration=<0 0 -100>)
   &GravityForce(Gravity=&RandomRangeVector(Minimum=<0 0 -80>, Maximum=<1 1 -40>))
   &VectorNoiseForce(Module.Use Random Noise Time Interval=false)
   &SolveForcesAndVelocity(Clamp Velocity=false, Limit Acceleration=false)
   &ScaleSpriteSize(Scale Factor=&Vector2DFromFloat(Value=&FloatFromCurve(FloatCurve=DataInterface() {propertyValues {Curve {keys [
          {arriveTangent=55.4408, interpMode=RCIM_Cubic, time=0, value=0}
          <at 0.03 1 RCIM_Cubic>
          {arriveTangent=-3.3237, leaveTangent=-3.3237, time=1, value=0}
         ]}}})))
   &ScaleColor(
    Scale Alpha=&FloatFromCurve(FloatCurve=DataInterface() {propertyValues {Curve {keys [{arriveTangent=-6.8352, interpMode=RCIM_Cubic, time=0, value=2}, <at 0.2 1>]}}})
    Scale RGB=&VectorFromCurve(VectorCurve=DataInterface() {propertyValues {
       XCurve {keys [<at 0 16 RCIM_Cubic>,<at 0.8 1 RCIM_Cubic>]}
       YCurve {keys [<at 0 16 RCIM_Cubic>,<at 0.8 0.1249 RCIM_Cubic>]}
       ZCurve {keys [<at 0 16 RCIM_Cubic>,<at 0.8 0.2547 RCIM_Cubic>]}
      }})
    ScaleA=true
    ScaleRGB=true
   )
   &ScaleSpriteSizeBySpeed(Max Scale Factor=<1 50>, Source Velocity:Vector3f=&Multiply_Vector(A=<0 0.015 1>, B=$Particles.Velocity))
  ]
  rendererClasses [&NiagaraSpriteRendererProperties]
  renderers [{rendererClass=&NiagaraSpriteRendererProperties}]
 }
 SparksStart2 {
  simTarget=GPUComputeSim
  emitterUpdateScript [
   &EmitterState(Life Cycle Mode=::NewEnumerator1, Loop Duration=1, UseLoopDelay=true)
   &SpawnBurst_Instantaneous(
    Spawn Count=&RandomRangeInt(Minimum=35, Maximum=60)
    Spawn Probability=$User.SpawnSpark
    Use Spawn Probability=true
   )
   &SpawnRate(SpawnRate=&Multiply_Float(A=90, B=$User.SpawnSpark), Use Spawn Probability=false)
  ]
  particleSpawnScript [
   &InitializeParticle(
    Lifetime=&RandomRangeFloat(Minimum=0.12, Maximum=0.48, Evaluation Type=::NewEnumerator1)
    Module.Write Color=true
    Module.Write Lifetime=true
    Module.Write Mass=true
    Module.Write Material Random=true
    Module.Write Position=false
    Module.Write Scale=false
    Module.Write SpriteRotation=false
    Module.Write SpriteSize=true
    Position=&RandomRangeVector(Minimum=<-22 0 0>, Maximum=<22 0 0>, Evaluation Type=::NewEnumerator1)
    Sprite Size=&RandomRangeVector2D(Minimum=<0.5 0.5>, Maximum=<1 1>, Evaluation Type=::NewEnumerator1)
   )
   &SphereLocation(
    Hemisphere X=true
    MaskSpawnGroup=false
    Module.Override Randomness=false
    Sphere Radius=&RandomRangeFloat(Minimum=5, Maximum=15, Evaluation Type=::NewEnumerator1)
    Surface Only=false
   )
   &Color(Color=&MakeLinearColorFromVectorAndFloat(
     "Vector (RGB)"=&MakeVectorFromLinearColorRGB(Color=<rgba 5 1.6977 0.0741 1>)
     "Float (Alpha)"=&ClampFloat(Float=1)
    ))
   &SystemLocation(MaskSpawnGroup=false, Offset=&RandomRangeVector(Minimum=<-1.5 -12 5>, Maximum=<1.5 -12 5>))
   &AddVelocityFromPoint(
    Module.Use Falloff Distance=false
    Velocity Strength=&FloatFromCurve(
     FloatCurve=DataInterface() {propertyValues {Curve {keys [{interpMode=RCIM_Cubic, tangentMode=RCTM_Break, time=0, value=0.3423}, <at 1 1>]}}}
     CurveIndex=$Emitter.Age
     Scale Curve=&RandomRangeFloat(Minimum=-90, Maximum=-275)
    )
   )
   &AddVelocity(Velocity=&RandomRangeVector(Minimum=<0 0 30>, Maximum=<0 0 80>, Evaluation Type=::NewEnumerator1)) {module=AddVelocity001}
   &AddVelocity(Velocity=&Multiply_VectorByFloat(
     Vector=$Output.SphereLocation.SphereVector
     Float=&RandomRangeFloat(Minimum=-20, Maximum=30, Evaluation Type=::NewEnumerator1)
    )) {module=AddVelocity002}
  ]
  particleUpdateScript [
   &ParticleState()
   &AccelerationForce(Acceleration=<0 0 -100>)
   &GravityForce(Gravity=&RandomRangeVector(Minimum=<0 0 -600>, Maximum=<1 1 -200>))
   &SolveForcesAndVelocity(Clamp Velocity=false, Limit Acceleration=false)
   &ScaleSpriteSize(Scale Factor=&Vector2DFromFloat(Value=&FloatFromCurve(FloatCurve=DataInterface() {propertyValues {Curve {keys [
          {arriveTangent=55.4408, interpMode=RCIM_Cubic, time=0, value=0}
          <at 0.03 1 RCIM_Cubic>
          {arriveTangent=-3.3237, leaveTangent=-3.3237, time=1, value=0}
         ]}}})))
   &ScaleColor(
    Scale Alpha=&FloatFromCurve(FloatCurve=DataInterface() {propertyValues {Curve {keys [{arriveTangent=-6.8352, interpMode=RCIM_Cubic, time=0, value=2}, <at 0.2 1>]}}})
    Scale RGB=&VectorFromCurve(VectorCurve=DataInterface() {propertyValues {
       XCurve {keys [<at 0 16 RCIM_Cubic>,<at 0.8 1 RCIM_Cubic>]}
       YCurve {keys [<at 0 16 RCIM_Cubic>,<at 0.8 0.1249 RCIM_Cubic>]}
       ZCurve {keys [<at 0 16 RCIM_Cubic>,<at 0.8 0.2547 RCIM_Cubic>]}
      }})
    ScaleA=true
    ScaleRGB=true
   )
   &ScaleSpriteSizeBySpeed(Max Scale Factor=<1 50>, Source Velocity:Vector3f=&Multiply_Vector(A=<0 0.015 1>, B=$Particles.Velocity))
  ]
  rendererClasses [&NiagaraSpriteRendererProperties]
  renderers [{rendererClass=&NiagaraSpriteRendererProperties}]
 }
}
summary {
 systemName=NS_Zipline_Pulley_SpeedLines_Converted
 emitters [
  {emitterName=SkidSharp, rendererClasses [&NiagaraMeshRendererProperties]}
  {emitterName=SparksStart, simTarget=GPUComputeSim, rendererClasses [&NiagaraSpriteRendererProperties]}
  {emitterName=SkidSharp001, rendererClasses [&NiagaraMeshRendererProperties]}
  {emitterName=Light, rendererClasses [&NiagaraLightRendererProperties]}
  {emitterName=SparksStart2, simTarget=GPUComputeSim, rendererClasses [&NiagaraSpriteRendererProperties]}
  {emitterName=Smoke, rendererClasses [&NiagaraSpriteRendererProperties]}
  {emitterName=Smoke001, rendererClasses [&NiagaraSpriteRendererProperties]}
  {emitterName=SparksStart001, simTarget=GPUComputeSim, rendererClasses [&NiagaraSpriteRendererProperties]}
  {emitterName=SparksStart002, simTarget=GPUComputeSim, rendererClasses [&NiagaraSpriteRendererProperties]}
 ]
 userVariables [User.SpawnSpark {type=NiagaraFloat, defaultValue=1, description=""}]
}
